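# Conversation database for the Grandeur supply specialist NPC (9CO1M), providing the greeting and top-level branches when a player talks to them.
# A single entry point feeds one Speech node, whose contextual lines use controller requirements for first-ever meeting, specific player outfits and per-match conversation count.
# Curve rows set priority for first-meeting and outfit lines.
# The speech branches through link nodes to shared NPC Library quests, basic services and a Ch5S1 quest conversation.
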ConversationDatabase /NPCLibrary/NPCs/Grandeur_9CO1M/NPCLibrary_Conversation_Grandeur_9CO1M
entry_points [{tag=Conversation.Entry.Grandeur_9CO1M, destinations [#1]}]
nodes [
 ConversationEntryPointNode_7 {
  id=#1
  class=ConversationEntryPointNode
  outputs [#2]
  properties {entryTag {tagName=Conversation.Entry.Grandeur_9CO1M}}
 }
 Speech_C_4 {
  id=#2
  class=Speech_C
  outputs [#3, #4, #5]
  properties {
   nodeName=""
   generalConfig {
    defaultMessage=Let me help improve your chances.
    contextualMessages [
     {
      message=Never underestimate a sharp design.
      contextRequirements [{requirement {class=FortControllerRequirement_IsFirstEverConversationWithNPC}}]
      priority {curve {rowName=Tandem.Speech.Priority.FirstEverInteraction}}
     }
     {
      message="So you are Hope. A fitting name, considering who you fight."
      contextRequirements [{requirement {
         class=FortControllerRequirement_HasCID
         properties {allowedCIds [&Character_VoidRedemption_Rebel]}
        }}]
      priority {curve {rowName=Tandem.Speech.Priority.OutfitReaction}}
     }
     {
      message=Your banana friend will be freed. I promise you.
      contextRequirements [{requirement {
         class=FortControllerRequirement_HasCID
         properties {allowedCIds [&Character_UndergroundRebel_Fashion]}
        }}]
      priority {curve {rowName=Tandem.Speech.Priority.OutfitReaction}}
     }
     {
      message=Let me help improve your chances.
      contextRequirements [{requirement {
         class=FortControllerRequirement_CheckNPCMatchConversationCount
         properties {nPCUniqueIdTag {tagName=AISpawnerData.Type.Specialist.Supply.9CO1M}}
        }}]
      priority {curve {rowName=None}}
     }
     {
      message=Need a helping hand?
      contextRequirements [{requirement {
         class=FortControllerRequirement_CheckNPCMatchConversationCount
         properties {nPCUniqueIdTag {tagName=AISpawnerData.Type.Specialist.Supply.9CO1M}}
        }}]
      priority {curve {rowName=None}}
     }
     {
      message=The Society lacks perspective. Let us show them some.
      contextRequirements [{requirement {
         class=FortControllerRequirement_CheckNPCMatchConversationCount
         properties {nPCUniqueIdTag {tagName=AISpawnerData.Type.Specialist.Supply.9CO1M}}
        }}]
      priority {curve {rowName=None}}
     }
    ]
   }
  }
 }
 ConversationLinkNode_2 {id=#5, class=ConversationLinkNode, properties {remoteEntryTag {tagName=Conversation.Entry.Ch5S1.D3.Q8}}}
 ConversationLinkNode_0 {id=#4, class=ConversationLinkNode, properties {remoteEntryTag {tagName=Conversation.Entry.BasicServices}}}
 ConversationLinkNode_1 {id=#3, class=ConversationLinkNode, properties {remoteEntryTag {tagName=Conversation.Entry.NPCLibrary.Quests}}}
]
